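# Invisible material for skeletal meshes in the experimental Version 3 vehicle cosmetics parent set, going by its name and path.
# It uses the Masked blend mode with Unlit shading and is flagged for skeletal mesh use.
# The graph holds a single default zero Constant and no textures or parameters, presumably to mask out every pixel, though no output connection is shown.
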Material /VehicleCosmetics/Materials/VehicleParents_Version3_Experimental/Materials/Invis
graph {properties {bUsedWithSkeletalMesh=true, blendMode=BLEND_Masked, shadingModel=MSM_Unlit, usedShadingModels=Unlit}, nodes [Constant()]}
