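# Experimental Default Lit parent material for vehicle bodies, holding unfinished parameter branches with none shown wired to a material output.
# A channel mask picks from the packed grain texture to blend low and high roughness parameters.
# The bent-normal/AO texture's RGB feeds both inputs of MF_BentNormalAO_SimiHel, whose AO goes to a vertex interpolator and a checker-pattern lerp.
# Metallic and diffuse parameters stand unconnected.
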Material /VehicleCosmetics/Materials/VehicleParents_Version3_Experimental/Materials/MA_BodyBase
graph {
 properties {usedShadingModels=Default Lit}
 nodes [
  Constant() {r=1}
  Constant()
  n1 = TextureSampleParameter2D() {
   parameterName=PackedComp
   texture=/VehicleCosmetics/Materials/VehicleParents_Version3_Experimental/Materials/Textures/Grain/T_Grain_BodyTextures
   channelNames {a=Rough, b=Met, g=A0, r=Cureature}
  }
  LinearInterpolate(
   A=ScalarParameter() {parameterName=BRoughLL}
   B=ScalarParameter() {defaultValue=1, parameterName=BRoughLH}
   Alpha=ChannelMaskParameter(None=@n1.RGBA) {parameterName=RoughChannel}.R
  )
  n2 = TextureSampleParameter2D() {
   parameterName=BNAO
   samplerType=SAMPLERTYPE_LinearColor
   texture=/VehicleCosmetics/Materials/VehicleParents_Version3_Experimental/Materials/Textures/BNAO/Grain/T_GrainBNAO_01
   channelNames {a=AO, b=BNB, g=BNG, r=BNR}
  }
  n3 = MaterialFunctionCall!(Normal=@n2.RGB, AOIn=@n2.RGB) {materialFunction=/VehicleCosmetics/Materials/VehicleParents_Version3_Experimental/Materials/MaterialFunctions/MF_BentNormalAO_SimiHel}
  LinearInterpolate(
   A=@n3.AO
   Alpha=MaterialFunctionCall!(UVs=TextureCoordinate(), Tiling=Constant2Vector() {g=50, r=50}) {materialFunction=/Engine/ArtTools/RenderToTexture/MaterialFunctions/CheckerPattern}.Result
  )
  ScalarParameter() {defaultValue=1, parameterName=Met}
  VectorParameter() {parameterName=Dif}
  VertexInterpolator(VS=@n3.AO)
 ]
}
