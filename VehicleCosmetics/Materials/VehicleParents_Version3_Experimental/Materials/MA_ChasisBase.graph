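# Experimental base material for vehicle chassis, Default Lit and enabled for Nanite and skeletal meshes.
# Tinted base texture, flattenable normal and bent-normal AO join metallic and roughness taken from a packed MRA texture through channel-mask parameters and named reroutes.
# Glow adds a mask-driven BlackBody temperature to squared-colour emissive channels; another branch shapes an RGB-masked emissive with inverted Fresnel.
# Several nodes, including checker and UV experiments, feed nothing visible.
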Material /VehicleCosmetics/Materials/VehicleParents_Version3_Experimental/Materials/MA_ChasisBase
graph {
 properties {
  bNormalCurvatureToRoughness=true
  bUsedWithMeshDeformer=true
  bUsedWithNanite=true
  bUsedWithSkeletalMesh=true
  usedShadingModels=Default Lit
 }
 nodes [
  n1 = TextureSampleParameter2D() {parameterName=Emis2, texture=/VehicleCosmetics/Materials/VehicleParents_Version3_Experimental/Materials/Textures/Grain/T_Grain_Chassis_M}
  n2 = ScalarParameter() {defaultValue=0.28, parameterName=BBadd}
  Add(A=@n1.B, B=@n2)
  Add()
  Constant() {r=0.18}
  Constant()
  Constant() {r=1}
  Constant() {r=1}
  n3 = TextureSampleParameter2D() {
   group=Surface
   parameterName=MRA
   texture=/VehicleCosmetics/Materials/VehicleParents_Version3_Experimental/Materials/Textures/Grain/T_Grain_ChassisTextures
   channelNames {a="", b=AO, g=Roughness, r=Metallic}
  }
  LinearInterpolate(B=ScalarParameter() {parameterName=CC}, Alpha=@n3.Metallic)
  n4 = TextureSampleParameter2D() {
   parameterName=BNAO
   samplerType=SAMPLERTYPE_LinearColor
   texture=/VehicleCosmetics/Materials/VehicleParents_Version3_Experimental/Materials/Textures/BNAO/Grain/T_GrainBNAO_01
  }
  n5 = MaterialFunctionCall!(Normal=@n4.RGB, AOIn=@n4.RGB) {materialFunction=/VehicleCosmetics/Materials/VehicleParents_Version3_Experimental/Materials/MaterialFunctions/MF_BentNormalAO_SimiHel}
  LinearInterpolate(
   A=@n5.AO
   Alpha=MaterialFunctionCall!(UVs=TextureCoordinate(), Tiling=Constant2Vector() {g=50, r=50}) {materialFunction=/Engine/ArtTools/RenderToTexture/MaterialFunctions/CheckerPattern}.Result
  )
  LinearInterpolate(A=ScalarParameter() {parameterName=MetLL}, B=ScalarParameter() {defaultValue=1, parameterName=MetLH}, Alpha=NamedRerouteUsage())
  LinearInterpolate(A=ScalarParameter() {parameterName=RoughLL}, B=ScalarParameter() {defaultValue=1, parameterName=RoughLH}, Alpha=NamedRerouteUsage())
  MaterialFunctionCall(In=@n3.AO, Power=ScalarParameter() {defaultValue=1, parameterName=AOCurve}) {materialFunction=&SCurve}
  MaterialFunctionCall(In=@n1.B, Power=ScalarParameter() {defaultValue=0.5363, parameterName=BBCurve}) {materialFunction=&SCurve}
  Multiply(A=@n5.AO, B=LinearInterpolate(A=ScalarParameter() {parameterName=TexAOLL}, Alpha=NamedRerouteUsage()) {constA=0.5})
  NamedRerouteDeclaration(None=ChannelMaskParameter(None=@n3.RGBA) {maskChannel=Blue, parameterName=AO, defaultValue=<rgba 0 0 1 0>}.R) {name=AO, nodeColor=<rgba 0 0 1 1>}
  NamedRerouteDeclaration(None=ChannelMaskParameter(None=@n3.RGBA) {maskChannel=Blue, parameterName=Met, defaultValue=<rgba 0 0 1 0>}.R) {name=Metallic, nodeColor=<rgba 0 0.3672 1 1>}
  NamedRerouteDeclaration(None=ChannelMaskParameter(None=@n3.RGBA) {maskChannel=Alpha, parameterName=Rough, defaultValue=<rgba 0 0 0 1>}.R) {name=Roughness, nodeColor=<rgba 0 1 0 1>}
  NamedRerouteUsage()
  NamedRerouteUsage()
  n6 = TextureSampleParameter2D() {parameterName=Emissive, samplerType=SAMPLERTYPE_Masks, texture=null}
  n7 = VectorParameter() {parameterName=EmisColorR, defaultValue=<rgba 0 0 0 1>}
  n8 = Multiply(A=@n7.RGB, B=@n7.RGB)
  n9 = VectorParameter() {parameterName=EmisColorG, defaultValue=<rgba 0 0 0 1>}
  n10 = Multiply(A=@n6.G, B=Multiply(A=@n9.RGB, B=@n9.RGB))
  n11 = VectorParameter() {parameterName=EmisColorB, defaultValue=<rgba 0 0 0 1>}
  n12 = Multiply(A=@n6.B, B=Multiply(A=@n11.RGB, B=@n11.RGB))
  n13 = Multiply(A=Add(A=Multiply(A=@n6.R, B=@n8), B=Add(A=@n10, B=@n12)), B=ScalarParameter() {group=BaseColor, parameterName=EmissivePower})
  Reroute(None=@n13)
  n14 = VectorParameter() {group=BaseColor, parameterName=Paint Tint, defaultValue=<rgba 1 1 1 1>}
  n15 = TextureSampleParameter2D() {
   group=BaseColor
   parameterName=Base Texture
   texture=/VehicleCosmetics/Materials/VehicleParents_Version3_Experimental/Materials/Textures/Grain/T_Grain_Chassis_BC
  }
  n16 = LinearInterpolate(A=@n14.RGB, B=Multiply(A=@n15.RGB, B=@n14.RGB), Alpha=@n14.RGB)
  n17 = NamedRerouteUsage()
  n18 = ScalarParameter() {defaultValue=1, parameterName=UseMetLerpDif}
  Reroute(None=LinearInterpolate(A=@n16, B=@n16, Alpha=LinearInterpolate(B=@n17, Alpha=@n18)))
  n19 = NamedRerouteUsage()
  n20 = StaticSwitchParameter(True=LinearInterpolate(A=@n19, B=Constant() {r=0.1}, Alpha=NamedRerouteUsage()), False=@n19) {defaultValue=true, group=Surface, parameterName=Override Metal Rough}
  n21 = ScalarParameter() {parameterName=RoughSimple}
  Reroute(None=LinearInterpolate(A=@n20, B=LinearInterpolate(A=@n20, B=@n21, Alpha=@n3.Metallic), Alpha=ScalarParameter() {parameterName=UseRoughIv}))
  n22 = OneMinus(None=Fresnel(ExponentIn=ScalarParameter() {defaultValue=0.1342, parameterName=EmisFresExp}) {baseReflectFraction=0})
  n23 = ScalarParameter() {parameterName=EmisCurve}
  Reroute(None=Multiply(
    A=@n13
    B=MaterialFunctionCall(In=@n22, Power=@n23) {materialFunction=&SCurve}.Result
   ))
  n24 = TextureSampleParameter2D() {
   group=Surface
   parameterName=Normal
   samplerType=SAMPLERTYPE_Normal
   texture=/VehicleCosmetics/Materials/VehicleParents_Version3_Experimental/Materials/Textures/Grain/T_Grain_Chassis_N
  }
  Reroute(None=LinearInterpolate(A=@n24.RGB, B=Constant3Vector() {constant=<rgba 0 0 1 0>}, Alpha=ScalarParameter() {parameterName=FlattenBaseNormal}))
  Reroute(None=@n5.AO)
  Reroute(None=Multiply(A=NamedRerouteUsage(), B=ScalarParameter() {group=Surface, parameterName=MetalPaintMetallic}))
  n25 = BlackBody!(Temp=Multiply(A=Power(Base=@n1.B, Exp=@n2), B=ScalarParameter() {defaultValue=4449.9189, parameterName=BB}))
  n26 = ScalarParameter() {defaultValue=1, parameterName=BBMulti}
  n27 = TextureSampleParameter2D() {parameterName=Emis, texture=/VehicleCosmetics/Materials/VehicleParents_Version3_Experimental/Materials/Textures/Grain/T_Grain_Chassis_E}
  n28 = VectorParameter() {parameterName=Red}
  n29 = Multiply(A=@n27.B, B=Multiply(A=@n28.RGB, B=@n28.RGB))
  n30 = VectorParameter() {parameterName=GReen', defaultValue=<rgba 1 1 1 100>}
  n31 = Multiply(A=@n27.G, B=Multiply(A=@n30.RGB, B=@n30.RGB))
  Reroute(None=Add(A=Multiply(A=@n25, B=@n26), B=Add(A=@n29, B=@n31)))
  ScalarParameter()
  Subtract(A=Multiply(A=@n4.RGB) {constB=2})
  TextureCoordinate() {coordinateIndex=2}
  TextureCoordinate() {coordinateIndex=1}
  TextureSampleParameter2D() {
   parameterName=LegSource
   texture=/VehicleCosmetics/Materials/VehicleParents_Version3_Experimental/Materials/Textures/Grain/T_Grain_BodyTextures
   channelNames {a=Rough, b=Met, g=A0, r=Cureature}
  }
  VertexInterpolator(VS=@n5.AO)
 ]
}
