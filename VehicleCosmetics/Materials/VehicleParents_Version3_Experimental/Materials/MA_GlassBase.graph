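# Experimental vehicle glass base material, translucent with Thin Translucent shading and per-pixel surface lighting.
# Static switches choose base color from a UV unit-square mask offset by Lens1 (named WorldBasedMask), vertex-color RGB tints, or flat Dif.
# An S-curved Fresnel blends that color toward Dif.
# A bent-normal texture yields AO with a SkipAO fade, while parameters expose roughness, metallic, specular, opacity, transmittance and coverage.
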Material /VehicleCosmetics/Materials/VehicleParents_Version3_Experimental/Materials/MA_GlassBase
graph {
 properties {
  blendMode=BLEND_Translucent
  shadingModel=MSM_ThinTranslucent
  translucencyLightingMode=TLM_SurfacePerPixelLighting
  usedShadingModels=Thin Translucent
 }
 nodes [
  Add()
  Constant() {r=1}
  Constant()
  n1 = VectorParameter() {parameterName=VcolorR}
  n2 = VectorParameter() {parameterName=VcolorG}
  n3 = VectorParameter() {parameterName=Lens1, defaultValue=<rgba 0 0 0 1>}
  n4 = Add(A=TextureCoordinate(), B=Multiply(A=AppendVector(A=@n3.R, B=@n3.R), B=Constant2Vector() {g=1, r=-1}) {constA=-1})
  n5 = Multiply(A=Add(A=@n4) {constB=0}, B=Add(A=@n4) {constB=-1})
  MaterialExpressionOneMinus_8 = OneMinus(None=Saturate(None=Ceil(None=@n5)))
  n6 = Convert(None=@MaterialExpressionOneMinus_8) {
   nodeName=Break Float2
   convertInputs [{
     type=Vector2
     expressionInput {
      expression=&self:MaterialExpressionOneMinus_8
      inputName=None
      mask=0
      maskA=0
      maskB=0
      maskG=0
      maskR=0
      outputIndex=0
     }
    }]
   convertOutputs [{type=Scalar}, {type=Scalar}]
  }
  n7 = Multiply(A=@n6, B=@n6)
  n8 = VertexColor()
  n9 = LinearInterpolate(
   A=LinearInterpolate(A=LinearInterpolate(B=@n1.RGB, Alpha=@n8.R), B=@n2.RGB, Alpha=@n8.G)
   B=VectorParameter() {parameterName=VcolorB}.RGB
   Alpha=@n8.B
  )
  n10 = VectorParameter() {parameterName=Dif}
  n11 = ScalarParameter() {parameterName=FresExp}
  LinearInterpolate(
   A=StaticSwitchParameter(
    True=LinearInterpolate(A=@n1.RGB, B=@n2.RGB, Alpha=@n7)
    False=StaticSwitchParameter(True=@n9, False=@n10.RGB) {parameterName=VertexColorMask}
   ) {parameterName=WorldBasedMask}
   B=@n10.RGB
   Alpha=MaterialFunctionCall(In=Fresnel(ExponentIn=@n11) {baseReflectFraction=0}, Power=ScalarParameter() {parameterName=Scurve}) {materialFunction=&SCurve}.Result
  )
  n12 = TextureSampleParameter2D() {
   parameterName=BNAO
   samplerType=SAMPLERTYPE_LinearColor
   texture=/VehicleCosmetics/Materials/VehicleParents_Version3_Experimental/Materials/Textures/BNAO/Grain/T_GrainBNAO_01
  }
  LinearInterpolate(
   A=MaterialFunctionCall!(Normal=@n12.RGB, AOIn=@n12.RGB) {materialFunction=/VehicleCosmetics/Materials/VehicleParents_Version3_Experimental/Materials/MaterialFunctions/MF_BentNormalAO_SimiHel}.AO
   Alpha=ScalarParameter() {parameterName=SkipAO}
  )
  ScalarParameter() {parameterName=BRough}
  ScalarParameter() {parameterName=Met}
  ScalarParameter() {defaultValue=0.5, parameterName=Spec}
  ScalarParameter() {defaultValue=0.5, parameterName=OP}
  ThinTranslucentMaterialOutput!(
   TransmittanceColor=VectorParameter() {parameterName=TransmColor, defaultValue=<rgba 1 1 1 1>}.RGB
   SurfaceCoverage=ScalarParameter() {defaultValue=1, parameterName=Coverage}
  )
 ]
}
