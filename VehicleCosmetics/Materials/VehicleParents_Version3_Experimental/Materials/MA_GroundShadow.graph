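# Translucent unlit material for a vehicle ground shadow, in an experimental Version 3 set of vehicle cosmetic parent materials.
# A grayscale shadow texture is shaped by a power and an S-curve, then drives a lerp between OPH and OPL, inverting the texture at default values.
# The saturated result is multiplied by a generated rounded-rectangle mask with adjustable size and sharpness, which bounds the shadow.
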Material /VehicleCosmetics/Materials/VehicleParents_Version3_Experimental/Materials/MA_GroundShadow
graph {
 properties {bUsedWithNanite=true, blendMode=BLEND_Translucent, shadingModel=MSM_Unlit, usedShadingModels=Unlit}
 nodes [
  Constant()
  n1 = ScalarParameter() {defaultValue=1, parameterName=OPH}
  n2 = ScalarParameter() {parameterName=OPL}
  n3 = TextureSample() {
   samplerType=SAMPLERTYPE_LinearGrayscale
   texture=/VehicleCosmetics/Materials/VehicleParents_Version3_Experimental/Materials/Textures/T_GroundShadow
  }
  n4 = MaterialFunctionCall(
   In=Power(Base=@n3.R, Exp=ScalarParameter() {defaultValue=2, parameterName=Power})
   Power=ScalarParameter() {defaultValue=25.8955, parameterName=Scurve}
  ) {materialFunction=&SCurve}
  Multiply(
   A=MaterialFunctionCall!(
    UV Coords
    Box Dimensions=ScalarParameter() {defaultValue=0.9026, parameterName=Dim}
    CornerRadius
    Sharpness=ScalarParameter() {parameterName=Sr}
    Center
   ) {materialFunction=/Engine/Functions/Engine_MaterialFunctions01/Gradient/GeneratedRoundRect}.Result
   B=Saturate(None=LinearInterpolate(A=@n1, B=@n2, Alpha=@n4.Result) {constA=1, constB=0})
  )
 ]
}
