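# Default Lit master material for a sports car wheel, flagged for Nanite, skeletal meshes and mesh deformers.
# A packed Masks texture supplies mask, remapped AO and curvature reroutes, used to blend two sets of base color, metallic and roughness parameters.
# Static switches can bypass masking or AO, alongside a normal map parameter.
# Bent normal and AO textures on UV1 feed MF_BentNormalAO_SimiHel, with a scalar blending in extra AO multiplication.
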Material /VehicleCosmetics/Materials/VehicleParents_Version3_Experimental/Materials/MA_SportsCar_Wheel
graph {
 properties {
  bUsedWithMeshDeformer=true
  bUsedWithNanite=true
  bUsedWithSkeletalMesh=true
  pixelDepthOffsetMode=PDOM_Legacy
  usedShadingModels=Default Lit
 }
 nodes [
  n1 = TextureCoordinate() {coordinateIndex=1}
  n2 = TextureSample(UVs=@n1) {
   samplerType=SAMPLERTYPE_LinearColor
   texture=/VehicleCosmetics/Materials/VehicleParents_Version3_Experimental/Materials/Textures/T_SportsCarWheel_BN
  }
  n3 = TextureSample(UVs=@n1) {
   samplerType=SAMPLERTYPE_LinearGrayscale
   texture=/VehicleCosmetics/Materials/VehicleParents_Version3_Experimental/Materials/Textures/T_SportsCarWheel_AO
  }
  n4 = NamedRerouteUsage()
  MaterialFunctionCall!(Normal=@n2.RGB, AOIn=LinearInterpolate(A=@n3.R, B=Multiply(A=@n4, B=@n3.R), Alpha=ScalarParameter() {parameterName=PakedAOMulti})) {materialFunction=/VehicleCosmetics/Materials/VehicleParents_Version3_Experimental/Materials/MaterialFunctions/MF_BentNormalAO_SimiHel}
  n5 = TextureSampleParameter2D() {
   parameterName=Masks
   samplerType=SAMPLERTYPE_LinearColor
   texture=/VehicleCosmetics/Materials/VehicleParents_Version3_Experimental/Materials/Textures/T_SportsCar_Wheel_M_4K
  }
  NamedRerouteDeclaration(None=@n5.R) {name=Mask, nodeColor=<rgba 0 1 0.7891 1>}
  NamedRerouteDeclaration(None=LinearInterpolate(Alpha=@n5.G) {constA=0.4}) {name=AO, nodeColor=<rgba 0 1 0 1>}
  NamedRerouteDeclaration(None=@n5.B) {name=Curvature, nodeColor=<rgba 0 0 1 1>}
  n6 = VectorParameter() {parameterName=BaseColor1, defaultValue=<rgba 1 1 1 1>}
  n7 = VectorParameter() {parameterName=BaseColor2, defaultValue=<rgba 1.5 1.5 1.5 1>}
  n8 = NamedRerouteUsage()
  StaticSwitchParameter(True=@n6.RGB, False=Multiply(A=LinearInterpolate(A=@n6.RGB, B=@n7.RGB, Alpha=@n8), B=NamedRerouteUsage())) {parameterName=No Mask}
  n9 = ScalarParameter() {defaultValue=1, parameterName=Metallic1}
  StaticSwitchParameter(True=@n9, False=LinearInterpolate(A=@n9, B=ScalarParameter() {defaultValue=1, parameterName=Metallic2}, Alpha=NamedRerouteUsage())) {parameterName=No Mask}
  n10 = ScalarParameter() {defaultValue=0.348, parameterName=Roughness1}
  StaticSwitchParameter(True=@n10, False=LinearInterpolate(A=@n10, B=ScalarParameter() {defaultValue=0.8819, parameterName=Roughness2}, Alpha=NamedRerouteUsage())) {parameterName=No Mask}
  StaticSwitchParameter(True=Constant3Vector() {constant=<rgba 1 1 1 0>}) {parameterName=No AO}
  TextureSampleParameter2D() {
   parameterName=Normal
   samplerType=SAMPLERTYPE_Normal
   texture=/VehicleCosmetics/Materials/VehicleParents_Version3_Experimental/Materials/Textures/T_SportsCar_Wheel_N_4K
  }
 ]
}
